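# Utility environment material that paints a tintable grid, projected top-down in world space so it ignores mesh UVs.
# World position is divided by a negated scale built from the Texture Size parameter, then its XY samples a grid mask texture.
# The mask lerps between a Gridline Darkness-dimmed copy of Grid BOX Color and the full color.
# Roughness is a separate parameter; Z scale is computed but dropped by the XY mask.
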Material /Game/Environments/World/Utility/M_CP_Grid_01
graph {
 properties {pixelDepthOffsetMode=PDOM_Legacy, usedShadingModels=Default Lit}
 nodes [
  Constant() {r=1}
  Constant()
  n1 = Constant3Vector() {constant=<rgba 1 1 1 0>}
  n2 = OneMinus(None=ScalarParameter() {defaultValue=0.5, parameterName=Gridline Darkness})
  n3 = VectorParameter() {parameterName=Grid BOX Color, defaultValue=<rgba 1 1 1 1>}
  n4 = MaterialFunctionCall(Float3=Constant3Vector() {constant=<rgba 1 1 256 1>}) {materialFunction=&BreakOutFloat3Components}
  n5 = Multiply(A=@n4.R, B=ScalarParameter() {defaultValue=512, parameterName=Texture Size})
  n6 = Abs(None=AppendVector(A=AppendVector(A=@n5, B=@n5), B=@n4.B))
  n7 = Add(A=Divide(A=WorldPosition() {worldPositionShaderOffset=WPT_ExcludeAllShaderOffsets}.XYZ, B=Multiply(A=@n6) {constB=-1})) {constB=0, desc=Offsets Procedural Texture 0.5}
  LinearInterpolate(
   A=Multiply(A=Multiply(A=@n1, B=@n2), B=@n3.RGB)
   B=@n3.RGB
   Alpha=TextureSample(UVs=ComponentMask(None=@n7) {g=true, r=true}) {texture=&/T_CP_Grid_01_M}.RGB
  )
  ScalarParameter() {defaultValue=1, parameterName=Roughness}
 ]
}
