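# Default Lit utility material for environments that draws a world-aligned grid with a larger heat map grid aligned over it.
# World position XY, divided by each layer's texture size, samples grid mask textures, so the pattern ignores mesh UVs.
# Heat map offsets scale by the regular-to-heat-map size ratio, with a half-step shift, to align both grids.
# Nested lerps use both masks to blend a tintable box color with its darkened gridline version.
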Material /Game/Environments/World/Utility/M_CP_Grid_02_HeatMapAligned
graph {
 properties {pixelDepthOffsetMode=PDOM_Legacy, usedShadingModels=Default Lit}
 nodes [
  Constant()
  n1 = ScalarParameter() {defaultValue=0.5, group=GRID - Colors, parameterName=Gridline Darkness}
  n2 = VectorParameter() {group=GRID - Colors, parameterName=Grid BOX Color, defaultValue=<rgba 1 1 1 1>}
  n3 = Multiply(A=Multiply(A=Constant3Vector() {constant=<rgba 1 1 1 0>}, B=OneMinus(None=@n1)), B=@n2.RGB)
  n4 = WorldPosition() {worldPositionShaderOffset=WPT_ExcludeAllShaderOffsets}
  n5 = MaterialFunctionCall(Float3=Constant3Vector() {constant=<rgba 1 1 256 1>}) {materialFunction=&BreakOutFloat3Components}
  n6 = Multiply(A=@n5.R, B=ScalarParameter() {defaultValue=512, group=GRID - Regular, parameterName=Texture Size})
  n7 = AppendVector(A=@n6, B=@n6)
  n8 = Multiply(A=Abs(None=AppendVector(A=@n7, B=@n5.B))) {constB=-1}
  n9 = ComponentMask(None=Add(A=Divide(A=@n4.XYZ, B=@n8)) {constB=0, desc=Offsets Procedural Texture 0.5}) {g=true, r=true}
  n10 = MaterialFunctionCall(Float3=Constant3Vector() {constant=<rgba 1 1 256 1>}) {materialFunction=&BreakOutFloat3Components}
  n11 = Multiply(A=@n10.R, B=ScalarParameter() {defaultValue=4096, group=GRID - Heat Map, parameterName=Texture Size - Heat Map})
  n12 = Abs(None=AppendVector(A=AppendVector(A=@n11, B=@n11), B=@n10.B))
  n13 = Divide(
   A=Constant() {r=1}
   B=Divide(
    A=ScalarParameter() {group=GRID - Heat Map, parameterName=Texture Size - Heat Map}
    B=ScalarParameter() {group=GRID - Regular, parameterName=Texture Size}
   )
  )
  n14 = Divide(A=@n13)
  n15 = Add(A=Multiply(A=ScalarParameter() {group=GRID - Heat Map, parameterName=HeatMap Offset X}, B=@n13), B=@n14)
  n16 = Add(A=Multiply(A=ScalarParameter() {group=GRID - Heat Map, parameterName=HeatMap Offset Y}, B=@n13), B=@n14)
  n17 = Add(
   A=Divide(A=WorldPosition() {worldPositionShaderOffset=WPT_ExcludeAllShaderOffsets}.XYZ, B=Multiply(A=@n12) {constB=-1})
   B=AppendVector(A=AppendVector(A=@n15, B=@n16), B=Constant())
  ) {constB=0, desc=Offsets Procedural Texture 0.5}
  LinearInterpolate(
   A=@n3
   B=LinearInterpolate(A=@n3, B=@n2.RGB, Alpha=TextureSample(UVs=@n9) {texture=&/T_CP_Grid_01_M}.RGB)
   Alpha=TextureSample(UVs=ComponentMask(None=@n17) {g=true, r=true}) {texture=&T_CP_Grid_01_M}.RGB
  )
  ScalarParameter() {defaultValue=1, group=GRID - Colors, parameterName=Roughness}
 ]
}
